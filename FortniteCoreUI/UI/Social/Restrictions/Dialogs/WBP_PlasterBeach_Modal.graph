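# Social-restriction dialog (PlasterBeach) that presents a verification link and QR code alongside a status row.
# The URL comes from the social user view model's text or voice chat restriction extra info, shown as a hyperlink (plain text on consoles) and as a nearest-filtered QR texture where allowed.
# DoPlasterBeach switches between throbber and check rows, toggles dialog buttons, and calls VerifyPlasterBeach, which starts on activation and stops on deactivation.
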WidgetBlueprint /FortniteCoreUI/UI/Social/Restrictions/Dialogs/WBP_PlasterBeach_Modal
graphs [
 DoPlasterBeach {nodes [
   n1 = DoPlasterBeach!() -> (then, IsPlasterBeachActive:bool=~) {is=title K2Node_FunctionEntry}
   > SetActiveWidgetIndex(
    self:Common Visibility Switcher Object Reference=Get CommonVisibilitySwitcher_0() -> (CommonVisibilitySwitcher_0:Common Visibility Switcher Object Reference)
    Index=SelectInt(A=0, B=1, bPickA=@n1)
   )
   n2 = Get CommonVisibilitySwitcher_0()
   > n3 = For Loop(LastIndex=int - int(A=GetChildrenCount!(self:Panel Widget Object Reference=@n2) -> (ReturnValue:int=0)))
   n4 = "Equal (Integer)"(A=@n3, B=SelectInt(A=1, B=0, bPickA=Get IsPlasterBeachActive() -> (IsPlasterBeachActive:bool)))
   n5 = Get WBP_UIKit_Dialog_Internal() -> (WBP_UIKit_Dialog_Internal:WBP UIKit Dialog Internal Object Reference)
   > SetButtonEnabled!(execute, self:WBP UIKit Dialog Internal Object Reference=@n5, ButtonIndex:int=@n3, IsButtonEnabled:bool=@n4) -> (then)
   > SetButtonVisibility!(
    execute
    self:WBP UIKit Dialog Internal Object Reference=@n5
    ButtonIndex:int=@n3
    Visibility:ESlateVisibility Enum=Select(Option 0=Collapsed, Option 1=SelfHitTestInvisible, Index=@n4)
   ) -> (then)
   n6 = Get!(execute:Exec=@n3.Completed) -> (then:Exec, else:Exec, FortSocialUserInfoViewModel:Fort Social User Info View Model Object Reference)
   > VerifyPlasterBeach!(execute, self:Fort Social User Info View Model Object Reference=@n6, bDoPlasterBeach:bool=Get IsPlasterBeachActive()) -> (then)
  ]}
 EventGraph {nodes [
   Event Construct()
   > Parent: Construct()
   n1 = GetOwningPlayer()
   > SetSocialUser!(
    execute
    self:Fort Social User Viewmodel Object Reference=Get FortSocialUserInfoViewModel() -> (FortSocialUserInfoViewModel:Fort Social User Info View Model Object Reference)
    InSocialUser:Fort Social User Object Reference=GetSocialUserFromController!(FortPC:Fort Player Controller Object Reference=GetFortPlayerControllerFromActor!(Actor:Actor Object Reference=@n1) -> (ReturnValue:Fort Player Controller Object Reference)) -> (ReturnValue:Fort Social User Object Reference)
   ) -> (then)
   Event On Deactivated()
   > Parent: On Deactivated() {is=title K2Node_CallParentFunction}
   > n2 = Get()
   > VerifyPlasterBeach(self=@n2, bDoPlasterBeach=false)
   Event On Activated()
   > Parent: On Activated() {is=title K2Node_CallParentFunction}
   > SetDialogText()
   > SetDialogImage()
   > DoPlasterBeach!(execute, self, IsPlasterBeachActive:bool=true) -> (then) {is=title K2Node_CallFunction}
   > n3 = Get()
   > VerifyPlasterBeach(self=@n3, bDoPlasterBeach=true)
  ]}
 GetPlasterBeachInfo {nodes [
   GetPlasterBeachInfo() {is=title K2Node_FunctionEntry}
   > n1 = Get()
   > Return Node!(
    execute
    PlasterBeach:String=SelectString(
     A=Get TextChatRestrictionExtraInfo!(self:Fort Social User Info View Model Object Reference=@n1) -> (TextChatRestrictionExtraInfo:String)
     B=Get VoiceChatRestrictionExtraInfo!(self:Fort Social User Info View Model Object Reference=@n1) -> (VoiceChatRestrictionExtraInfo:String)
     bPickA=Get bBrigadeiroIsActiveAndEnabledOptionT!(self:Fort Social User Info View Model Object Reference=@n1) -> (bBrigadeiroIsActiveAndEnabledOptionT:bool)
    )
   )
   Return Node(execute=@n1.else)
  ]}
 GetPlasterBeachTextInjection {nodes [
   GetPlasterBeachTextInjection() {is=title K2Node_FunctionEntry}
   n1 = GetPlasterBeachInfo!(self) -> (PlasterBeach:String) {is=title K2Node_CallFunction}
   > Return Node!(
    execute
    Text:Text=SelectText(
     A="INVTEXT(\"example.com\")"
     B=Format Text!(Format="INVTEXT(\"<a id=\\\"url\\\" href=\\\"{PlasterBeach}\\\" style=\\\"Hyperlink\\\">example.com</>\")", PlasterBeach:String=@n1) -> (Result)
     bPickA=IsConsolePlatform!() -> (ReturnValue:bool)
    )
   )
  ]}
 SetDialogImage {nodes [
   SetDialogImage() {is=title K2Node_FunctionEntry}
   > n1 = GenerateQRCodeTextureFromString!(execute, StringToEncode:String=GetPlasterBeachInfo()) -> (then, ReturnValue:Texture 2D Object Reference)
   > Set Filter!(execute, Filter:TextureFilter Enum=TF_Nearest, self:Texture Object Reference=@n1) -> (then, Output_Get:TextureFilter Enum=TF_Nearest)
   n2 = Get Image_83() -> (Image_83:Image Object Reference)
   > SetBrushFromTexture!(execute, self:Image Object Reference=@n2, Texture:Texture 2D Object Reference=@n1, bMatchSize:bool=false) -> (then)
   > SetVisibility(
    self=@n2
    InVisibility=Select(
     Option 0=Collapsed
     Option 1=Visible
     Index=Is Plaster Beach Platform Image Allowed!() -> (IsImageAllowed:bool=~) {is=title K2Node_MacroInstance}
    )
   )
  ]}
 SetDialogText {nodes [
   SetDialogText() {is=title K2Node_FunctionEntry}
   > n1 = Get!(execute, self:WBP UIKit Dialog Internal Object Reference=Get WBP_UIKit_Dialog_Internal()) -> (then, else, UIKitDialogViewModel:UI Kit Dialog Viewmodel Object Reference)
   > SetText(
    self:Fort Rich Text Block Legacy Object Reference=Get Text_Description() -> (Text_Description:Fort Rich Text Block Legacy Object Reference)
    InText:Text=Format Text(
     Format=Get DialogInfoText!(self:UI Kit Dialog Viewmodel Object Reference=@n1) -> (DialogInfoText:Text)
     PlasterBeach:Text=GetPlasterBeachTextInjection!(self) -> (Text:Text) {is=title K2Node_CallFunction}
    )
   )
  ]}
]
variables []
widgets {
 info {
  inheritedWidgetCount=0
  namedSlotCount=2
  parentClass=&CommonActivatableWidget
  rootWidgetClass=/FortUILibrary/Systems/Dialog/WBP_UIKit_Dialog_Internal.WBP_UIKit_Dialog_Internal_C
  widgetCount=15
 }
 widgets [
  WBP_UIKit_Dialog_Internal {
   widgetClassPath=/FortUILibrary/Systems/Dialog/WBP_UIKit_Dialog_Internal
   bIsVariable=true
   widget=WBP_UIKit_Dialog_Internal
   properties {
    designTimePreview_DialogVM=&/DialogVM_BrigadeiroPlasterBeach.DialogVM_BrigadeiroPlasterBeach_C
    useText=None
    buttons {
     each Button Horizontal Fill=true
     each Button Min Width=0
     each Button Size=Medium
     focuse Button=0
     space Between Buttons=30
     cTA Buttons Types [Loud,Regular]
     group Placement {
      horizontal Alignment=HAlign_Right
      vertical Alignment=VAlign_Bottom
      grid {column=1, column Span=0, layer=0, row=2, row Span=0, nudge=<0 0>}
      padding=<ltrb 80 64 80 48>
     }
    }
    dialogSIze {
     height Override {enable=false, value=800}
     max Aspect Ratio {enable=false, value=0}
     max Desired Height {enable=true, value=800}
     max Desired Width {enable=false, value=0}
     min Aspect Ratio {enable=false, value=0}
     min Desired Height {enable=true, value=720}
     min Desired Width {enable=false, value=0}
     width Override {enable=true, value=1200}
    }
   }
  }
  Overlay_0 {
   widgetClassPath=&Overlay
   namedSlotHost=WBP_UIKit_Dialog_Internal
   widget=Overlay_0
   properties {bOverride_Cursor=true, visibility=Visible}
  }
  VerticalBox_0 {
   widgetClassPath=&VerticalBox
   parent=Overlay_0
   widget=VerticalBox_0
   slot {class=OverlaySlot, properties {horizontalAlignment=HAlign_Fill, verticalAlignment=VAlign_Fill, padding=<ltrb 82 0 82 64>}}
  }
  HorizontalBox_155 {widgetClassPath=&HorizontalBox, parent=VerticalBox_0, widget=HorizontalBox_155, slot {class=VerticalBoxSlot}}
  VerticalBox_83 {
   widgetClassPath=&VerticalBox
   parent=HorizontalBox_155
   widget=VerticalBox_83
   slot {class=HorizontalBoxSlot, properties {size {sizeRule=Fill, value=1}}}
  }
  Text_Description {
   widgetClassPath=/Script/FortniteUI.FortRichTextBlockLegacy
   bIsVariable=true
   parent=VerticalBox_83
   widget=Text_Description
   properties {
    allowContextMenu=false
    autoWrapText=true
    disableTouchInput=true
    styleSet=&/PlasterBeachRichTextStyles
    text="Lorem ipsum dolor sit amet, consectetur adipiscing elit. In tincidunt justo velit, id fermentum ipsum euismod sit amet. Praesent ac viverra nisl. Curabitur varius tortor eget mauris viverra dapibus. Ut in ante mauris. Vivamus commodo ex nec sem eleifend tincidunt.\r\n\r\n- Lorem ipsum dolor sit amet.\r\n- Lorem ipsum dolor sit amet, {brigadeiro}\r\n\r\nVivamus eget interdum odio. Phasellus eget imperdiet nibh. Curabitur aliquam elit vitae risus eleifend, vel maximus libero ultrices. Nullam ac dapibus velit, sed malesuada est. Ut sed facilisis risus, at tincidunt lectus. Ut finibus est id metus semper maximus."
    visibility=SelfHitTestInvisible
    hyperlinkButtonStyle {
     clickedSlateSound {resourceObject=null}
     disabled {
      bIsDynamicallyLoaded=false
      drawAs=NoDrawType
      imageType=NoImage
      mirroring=NoMirror
      resourceName=None
      resourceObject=null
      tiling=NoTile
      imageSize=<0 0>
      margin=<ltrb 0 0 0 0>
      outlineSettings {
       bUseBrushTransparency=false
       roundingType=HalfHeightRadius
       width=0
       color {colorUseRule=UseColor_Specified, specifiedColor=<rgba 0 0 0 0>}
       cornerRadii=<0 0 0 0>
      }
      tintColor {colorUseRule=UseColor_Specified, specifiedColor=<rgba 1 1 1 1>}
      uVRegion {bIsValid=false, max=<0 0>, min=<0 0>}
     }
     disabledForeground {colorUseRule=UseColor_Foreground, specifiedColor=<rgba 1 0 1 1>}
     hovered {
      bIsDynamicallyLoaded=false
      drawAs=Border
      imageType=NoImage
      mirroring=NoMirror
      resourceName=None
      resourceObject=null
      tiling=NoTile
      imageSize=<8 8>
      margin=<ltrb 0 0 0 1>
      outlineSettings {
       bUseBrushTransparency=false
       roundingType=HalfHeightRadius
       width=0
       color {colorUseRule=UseColor_Specified, specifiedColor=<rgba 0 0 0 0>}
       cornerRadii=<0 0 0 0>
      }
      tintColor {colorUseRule=UseColor_Specified, specifiedColor=<rgba 0.9301 1 0.0103 1>}
      uVRegion {bIsValid=false, max=<0 0>, min=<0 0>}
     }
     hoveredForeground {colorUseRule=UseColor_Foreground, specifiedColor=<rgba 1 0 1 1>}
     hoveredSlateSound {resourceObject=null}
     normal {
      bIsDynamicallyLoaded=false
      drawAs=Image
      imageType=NoImage
      mirroring=NoMirror
      resourceName=None
      resourceObject=null
      tiling=NoTile
      imageSize=<32 32>
      margin=<ltrb 0 0 0 0>
      outlineSettings {
       bUseBrushTransparency=false
       roundingType=HalfHeightRadius
       width=0
       color {colorUseRule=UseColor_Specified, specifiedColor=<rgba 0 0 0 0>}
       cornerRadii=<0 0 0 0>
      }
      tintColor {colorUseRule=UseColor_Specified, specifiedColor=<rgba 1 1 1 0>}
      uVRegion {bIsValid=false, max=<0 0>, min=<0 0>}
     }
     normalForeground {colorUseRule=UseColor_Foreground, specifiedColor=<rgba 1 0 1 1>}
     normalPadding=<ltrb 0 0 0 0>
     pressed {
      bIsDynamicallyLoaded=false
      drawAs=NoDrawType
      imageType=NoImage
      mirroring=NoMirror
      resourceName=None
      resourceObject=null
      tiling=NoTile
      imageSize=<32 32>
      margin=<ltrb 0 0 0 0>
      outlineSettings {
       bUseBrushTransparency=false
       roundingType=HalfHeightRadius
       width=0
       color {colorUseRule=UseColor_Specified, specifiedColor=<rgba 0 0 0 0>}
       cornerRadii=<0 0 0 0>
      }
      tintColor {colorUseRule=UseColor_Specified, specifiedColor=<rgba 1 1 1 1>}
      uVRegion {bIsValid=false, max=<0 0>, min=<0 0>}
     }
     pressedForeground {colorUseRule=UseColor_Foreground, specifiedColor=<rgba 1 0 1 1>}
     pressedPadding=<ltrb 0 0 0 0>
     pressedSlateSound {resourceObject=null}
    }
   }
   slot {class=VerticalBoxSlot}
  }
  Image_83 {
   widgetClassPath=&Image
   bIsVariable=true
   parent=HorizontalBox_155
   widget=Image_83
   properties {brush {
     bIsDynamicallyLoaded=false
     drawAs=Image
     imageType=NoImage
     mirroring=NoMirror
     resourceName=None
     resourceObject=null
     tiling=NoTile
     imageSize=<256 256>
     margin=<ltrb 0 0 0 0>
     outlineSettings {
      bUseBrushTransparency=false
      roundingType=HalfHeightRadius
      width=0
      color {colorUseRule=UseColor_Specified, specifiedColor=<rgba 0 0 0 0>}
      cornerRadii=<0 0 0 0>
     }
     tintColor {colorUseRule=UseColor_Specified, specifiedColor=<rgba 1 1 1 1>}
     uVRegion {bIsValid=false, max=<0 0>, min=<0 0>}
    }}
   slot {class=HorizontalBoxSlot, properties {horizontalAlignment=HAlign_Center, verticalAlignment=VAlign_Top, padding=<ltrb 80 0 32 0>}}
  }
  Overlay_1 {widgetClassPath=&Overlay, namedSlotHost=WBP_UIKit_Dialog_Internal, widget=Overlay_1}
  CommonVisibilitySwitcher_0 {
   widgetClassPath=&CommonVisibilitySwitcher
   bIsVariable=true
   parent=Overlay_1
   widget=CommonVisibilitySwitcher_0
   properties {bOverride_Cursor=true}
   slot {class=OverlaySlot, properties {verticalAlignment=VAlign_Bottom, padding=<ltrb 82 0 0 48>}}
  }
  HorizontalBox_0 {
   widgetClassPath=&HorizontalBox
   parent=CommonVisibilitySwitcher_0
   widget=HorizontalBox_0
   slot {class=CommonVisibilitySwitcherSlot, properties {verticalAlignment=VAlign_Center}}
  }
  WBP_UIKit_Throbber {
   widgetClassPath=/FortUILibrary/UIKit/Throbber/WBP_UIKit_Throbber
   bIsVariable=true
   parent=HorizontalBox_0
   widget=WBP_UIKit_Throbber
   slot {class=HorizontalBoxSlot}
  }
  CommonTextBlock_290 {
   widgetClassPath=&CommonTextBlock
   bIsVariable=true
   parent=HorizontalBox_0
   widget=CommonTextBlock_290
   properties {
    text=Lorem ipsum dolor sit amet
    font {
     bForceMonospaced=false
     bMaterialIsStencil=false
     fontMaterial=null
     fontObject=&HeadingNow
     letterSpacing=0
     monospacedWidth=1
     size=15.75
     skewAmount=0
     typefaceFontName=74 Regular
     outlineSettings {
      bApplyOutlineToDropShadows=false
      bMiteredCorners=false
      bSeparateFillAlpha=false
      outlineMaterial=null
      outlineSize=0
      outlineColor=<rgba 0 0 0 1>
     }
    }
    shadowOffset=<0 0>
   }
   slot {class=HorizontalBoxSlot, properties {verticalAlignment=VAlign_Center, padding=<ltrb 8 0 0 0>}}
  }
  HorizontalBox_1 {
   widgetClassPath=&HorizontalBox
   parent=CommonVisibilitySwitcher_0
   widget=HorizontalBox_1
   slot {class=CommonVisibilitySwitcherSlot}
  }
  WBP_UIKit_StatusIndicator {
   widgetClassPath=/FortUILibrary/UIKit/StatusIndicator/WBP_UIKit_StatusIndicator
   bIsVariable=true
   parent=HorizontalBox_1
   widget=WBP_UIKit_StatusIndicator
   properties {icon=/FortUILibrary/Textures/MaterialInstances/MI_UIKit_Icon_Check, overrideStatusIndicatorVisibility=true}
   slot {class=HorizontalBoxSlot, properties {padding=<ltrb 8 8 8 8>}}
  }
  CommonTextBlock {
   widgetClassPath=&CommonTextBlock
   bIsVariable=true
   parent=HorizontalBox_1
   widget=CommonTextBlock
   properties {
    text="Vestibulum mauris at, pulvinar ante. "
    font {
     bForceMonospaced=false
     bMaterialIsStencil=false
     fontMaterial=null
     fontObject=&HeadingNow
     letterSpacing=0
     monospacedWidth=1
     size=15.75
     skewAmount=0
     typefaceFontName=74 Regular
     outlineSettings {
      bApplyOutlineToDropShadows=false
      bMiteredCorners=false
      bSeparateFillAlpha=false
      outlineMaterial=null
      outlineSize=0
      outlineColor=<rgba 0 0 0 1>
     }
    }
    shadowOffset=<0 0>
   }
   slot {class=HorizontalBoxSlot, properties {verticalAlignment=VAlign_Center, padding=<ltrb 12 0 0 0>}}
  }
 ]
}
